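; Player controller component that opens and closes the FM Calibration modal, configured from playlist gameplay tags or command-line calls.
; After a startup delay it can auto-popup for the primary player, optionally skipping if already calibrated.
; Toggle messages on the global event router launch a UI state chart, or queue it until playlist load if blocked.
; While open, it blocks emotes, mutes game and jam audio, and adds an input-mode tag.
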(event Custom|RequestOpenModal
  (params (CalibrationControllerComponent :type FMCalibration%20Controller%20Component%20Object%20Reference) (UsePreamble :type Boolean))
  ;[1 1 1] Register: Commandline calls
  (Variables|Settings|SetWantsPreamble UsePreamble)
  (CallFunction|SendToggleModalMessage (Components|GetOwner)))

(event Custom|RegisterPreamble
  (params (CalibrationControllerComponent :type FMCalibration%20Controller%20Component%20Object%20Reference) (PreambleGameplayTag :type Gameplay%20Tag%20Structure))
  ;[1 1 1] Register: Commandline calls
  (Variables|Settings|SetExperience%20Tag PreambleGameplayTag))

(event EventBeginPlay
  (|CallParentFunction)
  (Utilities|FlowControl|Sequence
    (:then_0
      ;[1 1 1] Register: Commandline calls
      (EventDispatchers|BindEventtoOnCalibrationRequestOpenModal (delegate Custom|RequestOpenModal))
      (EventDispatchers|BindEventtoOnCalibrationRegisterPreamble (delegate Custom|RegisterPreamble)))
    (:then_1
      ;[1 1 1] Wait for Playlist to load, then setup the Modal Settings.
      (bind (_gamestate _playlist) (Class|FortAsyncActionCurrentPlaylistReady|PlaylistDataReadyAsync)
        (:OnReady
          (bind _asfort_playlist_athena (Utilities|Casting|CastToFortPlaylistAthena _playlist))
          (CallFunction|ConfigureCalibrationModalfromPlaylist (Variables|/Script/FortniteGame.FortPlaylist|GetGameplayTagContainer _asfort_playlist_athena))
          (Variables|Runtime|SetPlaylistLoaded true)
          (if (Variables|Runtime|GetOpenModalQueued)
            (CallFunction|AttemptToLaunchModal)))))))

(event EventDisplayCalibrationWidget
  ;[0.06944 1 0.3937] ToggleModal: Listen for UI Message to open the Calibration Modal
  (Utilities|FlowControl|Sequence
    (:then_0
      (CallFunction|SetDelayPreambleTimer (delegate Custom|Delay%20Preamble)))
    (:then_1
      (GameplayEventRouter|StartListeningToEvent (GameplayEventRouter|GetGlobalEventRouter) "/Script/FMCalibrationRuntime.FMEvent_Calibration_ToggleModal"
        (:OnEventReceived
          (CallFunction|AttemptToLaunchModal))))))

(event Custom|Delay%20Preamble
  ;[0.06944 1 0.3937] ToggleModal: Listen for UI Message to open the Calibration Modal
  (CallFunction|HandleDelayFinished)
  (Default|AttemptToAutoLaunchModal))

(event EventEndPlay
  (params (EndPlayReason :type EEndPlayReason%20Enum :default "Destroyed"))
  (Default|CloseCalibrationModal false))

(fn AttempttoAutoLaunchModal
  ;[1 1 1] Only support calibrating on Primary Player
  (if (CallFunction|IsOwnerPrimaryPlayer)
    ;[1 1 1] Only popup the modal automatically if the settings have asked for it.
    (if (Variables|Settings|GetAutoPopup)
      ;[1 1 1] Always launch if the playlist wants a preamble or check to see if the skip auto popup is requested and the player has already calibrated.
      (if (or (Variables|Settings|GetWantsPreamble) (Math|Boolean|NANDBoolean (Variables|Settings|GetSkipAutoPopupIfCalibrated) (and (CallFunction|GetIsCalibrated) (CallFunction|IsCalibrationDataonLatestVersion))))
        (CallFunction|SendToggleModalMessage (Components|GetOwner))
        (else
          (CallFunction|OnCalibrationSkipped))))))

(fn AttempttoLaunchModal
  (if (Math|Boolean|ANDBoolean (CallFunction|IsModalReadyToOpen) (Variables|Runtime|GetStartupModalDelayFinished) (CallFunction|OwnerAllowedtoCalibrate))
    (CallFunction|OpenCalibrateModal)
    (else
      (Variables|Runtime|SetOpenModalQueued true))))

(fn BlockEmoting
  (params (IsBlocked :type Boolean))
  (bind _asfort_player_controller_athena (Utilities|Casting|CastToFortPlayerControllerAthena (CallFunction|GetOwningController)))
  (UI|SetPickerEnabled _asfort_player_controller_athena (not IsBlocked)))

(fn CloseCalibrationModal
  (params (CloseToQuestMenu :type Boolean))
  (if (>= (Variables|Default|GetCalibrationStateChartKey) 0)
    (UIChart|RemoveStateChart :self (GameInstanceSubsystems|GetUIStateChartManager) :Key (Variables|Default|GetCalibrationStateChartKey) :Urgency "Normal")
    (Utilities|FlowControl|Sequence
      (:then_0
        ; Re-allow the user to emote.
        (Default|BlockEmoting false))
      (:then_1
        (Default|UnmuteSounds))
      (:then_2
        (if (Variables|GetCloseToQuestMenu)
          (GameplayEventRouter|BroadcastEvent (GameplayEventRouter|GetGlobalEventRouter) (Utilities|Struct|MakeFMEventCalibrationModalClosetoQuest) 0)
          (else
            (GameplayEventRouter|BroadcastEvent (GameplayEventRouter|GetGlobalEventRouter) (Utilities|Struct|MakeFMEventCalibrationModalClosed) 0))))
      (:then_3
        (Input|RemoveTagfromInputMode (PlayerController|LocalPlayerSubsystems|GetEnhancedInputLocalPlayerSubsystem (CallFunction|GetOwningController)) (Variables|Default|GetCalibrationInputModeTag) "(bIgnoreAllPressedKeysUntilRelease=True,bForceImmediately=False,bNotifyUserSettings=False)"))
      (:then_4
        (Variables|Default|SetCalibrationStateChartKey -1)))))

(fn CloseQuestScreen
  (bind _returnvalue (GameplayEventRouter|GetEventRouter "/Script/FortniteGame.FortPlayspace" (CallFunction|GetOwningController)))
  (GameplayEventRouter|BroadcastEvent _returnvalue (Utilities|Struct|MakeShowSparksQuestScreenEvent false) 0))

(fn ConfigureCalibrationModalFromPlaylist
  (params (Playlist%20Tags :type Gameplay%20Tag%20Container%20Structure))
  (CallFunction|SetExperienceTagFromGameplayTags (Variables|GetPlaylist%20Tags))
  (CallFunction|SetUsePreambleFromGameplayTags (Variables|GetPlaylist%20Tags))
  (CallFunction|SetAutoPopupFromGameplayTags (Variables|GetPlaylist%20Tags))
  (CallFunction|SetSkipPopupifAlreadyCalibratedFromGameplayTags (Variables|GetPlaylist%20Tags)))

(fn ConsumeWantsPreamble
  (LocalVariables|SetUsePreamble (Variables|Settings|GetWantsPreamble))
  (Variables|Settings|SetWantsPreamble false)
  (return (LocalVariables|GetUsePreamble)))

(fn GetCalibrationHelper
  (Variables|Default|GetCalibrationHelper
    (:then
      _)
    (:else
      (bind _returnvalue (Game|SpawnActorfromClass "/FMCalibration/Actors/BP_FMCalibration_Helper.BP_FMCalibration_Helper_C" (Transformation|GetActorTransform (Pawn|GetControlledPawn (Fortnite|GetControllerfromActor (Components|GetOwner)))) "Undefined" "OverrideRootScale" (Pawn|GetControlledPawn (Fortnite|GetControllerfromActor (Components|GetOwner))) 0))
      (Variables|Default|SetCalibrationHelper _returnvalue)))
  (return (Variables|Default|GetCalibrationHelper)))

(fn HandleDelayFinished
  (Utilities|Time|ClearandInvalidateTimerbyHandle (Variables|Runtime|GetStartupModalTimerHandle))
  (Variables|Runtime|SetStartupModalDelayFinished true))

(fn IsModalReadytoOpen
  ;[1 1 1] Return False only if the Preamble is requested and the Experience is not currently registered.
  (return (Math|Boolean|NANDBoolean (not (GameplayTags|IsGameplayTagValid (Variables|Settings|GetExperience%20Tag))) (Variables|Settings|GetWantsPreamble))))

(fn MuteJamAudio
  (params (Mute :type Boolean))
  (bind _returnvalue (GameplayEventRouter|GetEventRouter "/Script/FortniteGame.FortPlayspace" (Components|GetOwner)))
  (Utilities|IsValid _returnvalue
    (:Is%20Valid
      (GameplayEventRouter|BroadcastEvent _returnvalue (Utilities|Struct|MakeFMCoreEventMuteJamAudioChanged Mute) 0))))

(fn MuteSounds
  (bind _returnvalue (Components|GetOwner))
  (bind _returnvalue_2 (Actor|GetComponentbyClass _returnvalue "/FMCalibration/Components/BP_FMCalibrationSoundMixProvider.BP_FMCalibrationSoundMixProvider_C"))
  ; Mute environmental sounds
  (Audio|PushSoundMixModifier (Variables|/FMCalibration/Components/BP_FMCalibrationSoundMixProvider.BP_FMCalibrationSoundMixProvider_C|GetMix _returnvalue_2))
  (Audio|Modulation|ActivateControlBusMix (Variables|/FMCalibration/Components/BP_FMCalibrationSoundMixProvider.BP_FMCalibrationSoundMixProvider_C|GetControlBusMix _returnvalue_2))
  (Default|MuteJamAudio true))

(fn OpenCalibrateModal
  (bind _returnvalue (UIChart|AddStateChart :self (GameInstanceSubsystems|GetUIStateChartManager) :StateChartAsset "/FMCalibration/Widgets/Scenes/UIState_FMCalibration.UIState_FMCalibration" :Urgency "Normal"))
  (Variables|Default|SetCalibrationStateChartKey _returnvalue)
  (GameplayEventRouter|BroadcastEvent (GameplayEventRouter|GetGlobalEventRouter) (Utilities|Struct|MakeFMEventCalibrationModalOpened) 0)
  (Utilities|FlowControl|Sequence
    (:then_0
      (Default|BlockEmoting true))
    (:then_1
      (Default|MuteSounds))
    (:then_2
      (Input|AddTagtoInputMode (PlayerController|LocalPlayerSubsystems|GetEnhancedInputLocalPlayerSubsystem (CallFunction|GetOwningController)) (Variables|Default|GetCalibrationInputModeTag) "(bIgnoreAllPressedKeysUntilRelease=True,bForceImmediately=False,bNotifyUserSettings=False)"))))

(fn OwnerAllowedToCalibrate
  (bind _asfort_player_controller (Utilities|Casting|CastToFortPlayerController (Components|GetOwner))
    (:then
      (bind _returnvalue (GameplayTags|HasAnyMatchingGameplayTags(Message) (Variables|/Script/Engine.Controller|GetPlayerState _asfort_player_controller) (GameplayTags|MakeGameplayTagContainerfromTag "(TagName=\"Mojo.FMCalibration.Modal.DisableCalibration\")")))
      (return (not _returnvalue)))
    (:CastFailed
      (return false))))

(fn SetAutoPopupfromGameplayTags
  (params (TagContainer :type Gameplay%20Tag%20Container%20Structure%20%28by%20ref%29))
  (Variables|Settings|SetAutoPopup (GameplayTags|HasTag TagContainer "(TagName=\"Mojo.FMCalibration.Modal.AutoPopup\")" true)))

(fn SetExperienceTagfromGameplayTags
  (params (Playlist%20Tags :type Gameplay%20Tag%20Container%20Structure%20%28by%20ref%29))
  (bind _array_element (Utilities|Array|ForEachLoopwithBreak (GameplayTags|BreakGameplayTagContainer (Variables|Default|GetSupportedPreambleTags)))
    (:LoopBody
      (if (GameplayTags|HasTag (Variables|GetPlaylist%20Tags) _array_element true)
        (Variables|Settings|SetExperience%20Tag _array_element)))))

(fn SetSkipPopupIfAlreadyCalibratedfromGameplayTags
  (params (TagContainer :type Gameplay%20Tag%20Container%20Structure%20%28by%20ref%29))
  (Variables|Settings|SetSkipAutoPopupIfCalibrated (GameplayTags|HasTag TagContainer "(TagName=\"Mojo.FMCalibration.Modal.SkipIfCal\")" true)))

(fn SetUsePreamblefromGameplayTags
  (params (TagContainer :type Gameplay%20Tag%20Container%20Structure%20%28by%20ref%29))
  (Variables|Settings|SetWantsPreamble (GameplayTags|HasTag TagContainer "(TagName=\"Mojo.FMCalibration.Modal.UsePreamble\")" true)))

(fn SetDelayPreambleTimer
  (params (Delegate :type Delegate))
  (bind _returnvalue (Utilities|Time|SetTimerbyEvent Delegate (ScalableFloat|GetValueatLevel (Variables|Default|GetPreambleDelay))))
  (Variables|Runtime|SetStartupModalTimerHandle _returnvalue))

(fn UnmuteSounds
  (bind _returnvalue (Components|GetOwner))
  (bind _returnvalue_2 (Actor|GetComponentbyClass _returnvalue "/FMCalibration/Components/BP_FMCalibrationSoundMixProvider.BP_FMCalibrationSoundMixProvider_C"))
  (Default|MuteJamAudio false)
  ; Unmute environmental sounds
  (Audio|PopSoundMixModifier (Variables|/FMCalibration/Components/BP_FMCalibrationSoundMixProvider.BP_FMCalibrationSoundMixProvider_C|GetMix _returnvalue_2))
  (Audio|Modulation|DeactivateControlBusMix (Variables|/FMCalibration/Components/BP_FMCalibrationSoundMixProvider.BP_FMCalibrationSoundMixProvider_C|GetControlBusMix _returnvalue_2)))
